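# Translucent unlit two-sided energy material for the Cherry backpack FX, built for Niagara ribbons and mesh particles, with a time-of-day tint times particle color.
# Opacity combines panning, noise-distorted smoke and tech-line textures in two offset copies split by vertex red and green as wings, shaped by UV and sphere masks.
# Panned distortion textures drive world position offset.
# Depth fade softens opacity, except on Forward and Mobile, which use scaled-down opacity.
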Material /BRCosmetics/Accessories/FORT_Backpacks/Backpack_F_MED_Cherry/FX/M_Backpack_Cherry_energy
graph {
 properties {
  bUsedWithNiagaraMeshParticles=true
  bUsedWithNiagaraRibbons=true
  blendMode=BLEND_Translucent
  pixelDepthOffsetMode=PDOM_Legacy
  shadingModel=MSM_Unlit
  twoSided=true
  usedShadingModels=Unlit
 }
 nodes [
  n1 = Panner(
   Coordinate=Add(
    A=DynamicParameter() {defaultValue=<rgba 0 1 1 10>, paramNames [Random_offset, Param2, Param3, WPO]}.Random_offset
    B=TextureCoordinate() {coordinateIndex=1}
   )
   Time=Multiply(A=Time(), B=ScalarParameter() {defaultValue=1.1, group=Panner Speed, parameterName=Speed})
  ) {speedX=-0.4}
  n2 = AppendVector(
   A=ComponentMask(None=TextureSample(UVs=@n1) {samplerType=SAMPLERTYPE_Normal, texture=&T_Weap_Distort_N_02}.RGB) {g=true, r=true}
   B=TextureSample(UVs=@n1) {texture=&T_Smooth_tile}.R
  )
  n3 = Add(A=ComponentMask(None=TextureCoordinate()) {r=true}, B=ScalarParameter() {defaultValue=-0.1, parameterName=WPO_Bottom_mask_offset})
  LinearInterpolate(
   A=Multiply(
    A=LinearInterpolate(B=@n2, Alpha=@n3)
    B=DynamicParameter() {defaultValue=<rgba 0 1 1 10>, paramNames [Random_offset, Param2, Param3, WPO]}.WPO
   ) {constB=10}
   Alpha=ScalarParameter() {desc=no_Wpo, group=Simple_Round_particle, parameterName=No_WPO}
  )
  n4 = VectorParameter() {parameterName=Color_Morning, defaultValue=<rgba 1 1 1 1>}
  n5 = VectorParameter() {parameterName=Color_Day, defaultValue=<rgba 1 1 1 1>}
  n6 = VectorParameter() {parameterName=Color_Evening, defaultValue=<rgba 1 1 1 1>}
  n7 = VectorParameter() {parameterName=Color_Night, defaultValue=<rgba 1 1 1 1>}
  n8 = ParticleColor()
  MaterialFunctionCall(Emissive=Multiply(
    A=MaterialFunctionCall(Color_Morning=@n4.RGB, Color_Day=@n5.RGB, Color_Evening=@n6.RGB, Color_Night=@n7.RGB) {materialFunction=&Color_Time_of_Day}.Result
    B=@n8.RGB
   )) {materialFunction=&MF_FX_EmissiveAdjustments}
  n9 = ComponentMask(None=TextureCoordinate()) {g=true}
  n10 = Multiply(A=@n9, B=OneMinus(None=@n9))
  n11 = VectorParameter() {parameterName=Texture_A_UVScale, defaultValue=<rgba 1 1 0 1>}
  n12 = Add(
   A=Multiply(A=TextureCoordinate() {uTiling=1.5}, B=AppendVector(A=@n11.R, B=@n11.R))
   B=AppendVector(A=DynamicParameter() {defaultValue=<rgba 0 1 1 10>, paramNames [Random_offset, Param2, Param3, WPO]}.Random_offset, B=Constant())
  )
  n13 = TextureCoordinate() {uTiling=0.4, vTiling=0.5}
  n14 = ComponentMask(None=TextureSample(UVs=Panner(Coordinate=@n13) {speedX=-0.2}) {samplerType=SAMPLERTYPE_LinearColor, texture=&T_NoisePack_LinearColor}.RGB) {g=true, r=true}
  n15 = Multiply(A=@n14) {constB=0.15}
  n16 = TextureSampleParameter2D(UVs=Panner(Coordinate=Add(A=@n12, B=@n15) {constB=0}) {speedX=-0.1}) {parameterName=Texture_A, samplerType=SAMPLERTYPE_Grayscale, texture=&T_Smoke_Trail_Soft}
  n17 = Multiply(A=@n16.R) {constB=1.8}
  n18 = Multiply(
   A=Power(Base=Multiply(A=@n10) {constB=3})
   B=Multiply(A=Clamp(None=@n17), B=ScalarParameter() {defaultValue=1, parameterName=Texture_A_Glow})
  ) {constA=1}
  n19 = Add(
   A=TextureCoordinate() {uTiling=1.5}
   B=AppendVector(A=DynamicParameter() {defaultValue=<rgba 0 1 1 10>, paramNames [Random_offset, Param2, Param3, WPO]}.Random_offset, B=Constant())
  )
  n20 = Multiply(A=@n14) {constB=0.2}
  n21 = TextureSampleParameter2D(UVs=Panner(Coordinate=Add(A=@n19, B=@n20)) {speedX=-0.3}) {parameterName=Texture_A, samplerType=SAMPLERTYPE_Grayscale, texture=&T_Smoke_Trail_Soft}
  n22 = Multiply(A=@n21.R, B=@n21.R)
  n23 = TextureSample() {samplerType=SAMPLERTYPE_Masks, texture=/Game/Characters/Player/Male/Medium/Bodies/M_MED_Storm_Soldier/Textures/ElectricityNoise_M}
  n24 = Clamp(None=Multiply(A=Multiply(A=@n22) {constB=2.3}, B=LinearInterpolate(Alpha=@n23.R) {constA=0.3}))
  n25 = ScalarParameter() {defaultValue=7, parameterName=Texture_A_intensity}
  n26 = TextureSample(UVs=TextureCoordinate() {uTiling=0.5, vTiling=0.5}) {samplerType=SAMPLERTYPE_Masks, texture=/Game/Weapons/WeaponSkins/Wraps/Textures/Mash/Mash_Noise}
  n27 = Multiply(A=Multiply(A=Multiply(A=@n24, B=@n25) {constB=7}, B=LinearInterpolate(Alpha=@n26.R) {constA=0.3})) {constB=2}
  n28 = TextureCoordinate() {uTiling=0.7, vTiling=0.5}
  n29 = LinearInterpolate(Alpha=TextureSample(UVs=Panner(Coordinate=@n28)) {samplerType=SAMPLERTYPE_Masks, texture=&Mash_Noise}.R) {constB=2}
  n30 = Add(A=@n27, B=Multiply(A=Power(Base=@n27), B=Clamp(None=@n29))) {constB=0}
  n31 = VectorParameter() {parameterName=Texture_C_UVScale, defaultValue=<rgba 1 1 0 1>}
  n32 = Add(
   A=Multiply(A=TextureCoordinate() {uTiling=1.2}, B=AppendVector(A=@n31.R, B=@n31.R))
   B=AppendVector(A=DynamicParameter() {defaultValue=<rgba 0 1 1 10>, paramNames [Random_offset, Param2, Param3, WPO]}.Random_offset, B=Constant())
  )
  n33 = Multiply(A=@n14) {constB=0.1}
  n34 = TextureSampleParameter2D(UVs=Panner(Coordinate=Add(A=@n32, B=@n33)) {speedX=-0.25}) {parameterName=Texture_C, texture=/Game/Effects/Fort_Effects/Textures/T_TechLines}
  n35 = ComponentMask(None=TextureCoordinate()) {r=true}
  n36 = Clamp(None=Multiply(A=Power(Base=@n35) {constExponent=3}) {constB=10})
  n37 = Multiply(A=Power(Base=Multiply(A=@n34.R) {constB=5}, Exp=ScalarParameter() {defaultValue=1, parameterName=Texture_C_power}), B=@n36) {constB=2}
  n38 = Constant()
  n39 = ScalarParameter() {parameterName=Not_USe_textureC}
  n40 = Multiply(A=Add(A=Add(A=@n18, B=@n30) {constB=0}, B=LinearInterpolate(A=@n37, B=@n38, Alpha=@n39)), B=VertexColor().R)
  n41 = ComponentMask(None=TextureCoordinate()) {g=true}
  n42 = Multiply(A=Multiply(A=@n41, B=OneMinus(None=@n41))) {constB=3}
  n43 = DynamicParameter() {defaultValue=<rgba 0 1 1 10>, paramNames [Random_offset, Param2, Param3, WPO]}
  n44 = Constant()
  n45 = Add(A=Add(A=TextureCoordinate(), B=AppendVector(A=@n43.Random_offset, B=@n44)), B=AppendVector(A=Constant() {r=0.5}, B=Constant()))
  n46 = TextureSampleParameter2D(UVs=Panner(Coordinate=Add(A=@n45, B=@n15) {constB=0}) {speedX=-0.1}) {parameterName=Texture_A, samplerType=SAMPLERTYPE_Grayscale, texture=&T_Smoke_Trail_Soft}
  n47 = Clamp(None=Multiply(A=@n46.R) {constB=1.8})
  n48 = ScalarParameter() {defaultValue=1, parameterName=Texture_A_Glow}
  n49 = DynamicParameter() {defaultValue=<rgba 0 1 1 10>, paramNames [Random_offset, Param2, Param3, WPO]}
  n50 = Constant()
  n51 = Add(A=Add(A=TextureCoordinate(), B=AppendVector(A=@n49.Random_offset, B=@n50)), B=AppendVector(A=Constant() {r=0.5}, B=Constant()))
  n52 = TextureSampleParameter2D(UVs=Panner(Coordinate=Add(A=@n51, B=@n20)) {speedX=-0.3}) {parameterName=Texture_A, samplerType=SAMPLERTYPE_Grayscale, texture=&T_Smoke_Trail_Soft}
  n53 = Multiply(A=@n52.R, B=@n52.R)
  n54 = TextureSample() {samplerType=SAMPLERTYPE_Masks, texture=&ElectricityNoise_M}
  n55 = Clamp(None=Multiply(A=Multiply(A=@n53) {constB=2.5}, B=LinearInterpolate(Alpha=@n54.R) {constA=0.3}))
  n56 = ScalarParameter() {defaultValue=7, parameterName=Texture_A_intensity}
  n57 = TextureSample(UVs=TextureCoordinate() {uTiling=0.5, vTiling=0.5}) {samplerType=SAMPLERTYPE_Masks, texture=&Mash_Noise}
  n58 = Multiply(A=Multiply(A=Multiply(A=@n55, B=@n56) {constB=7}, B=LinearInterpolate(Alpha=@n57.R) {constA=0.3})) {constB=2}
  n59 = Constant() {r=0.5}
  n60 = Constant()
  n61 = Panner(Coordinate=Add(A=TextureCoordinate() {uTiling=0.7, vTiling=0.5}, B=AppendVector(A=@n59, B=@n60))) {speedX=-0.05}
  n62 = Clamp(None=LinearInterpolate(Alpha=TextureSample(UVs=@n61) {samplerType=SAMPLERTYPE_Masks, texture=&Mash_Noise}.R) {constB=2})
  n63 = Add(A=Multiply(A=Power(Base=@n42), B=Multiply(A=@n47, B=@n48)) {constA=1}, B=Add(A=@n58, B=Multiply(A=@n58, B=@n62)))
  n64 = TextureCoordinate() {uTiling=1.2}
  n65 = VectorParameter() {parameterName=Texture_C_UVScale, defaultValue=<rgba 1 1 0 1>}
  n66 = AppendVector(A=@n65.R, B=@n65.R)
  n67 = DynamicParameter() {defaultValue=<rgba 0 1 1 10>, paramNames [Random_offset, Param2, Param3, WPO]}
  n68 = Constant()
  n69 = Add(A=Add(A=Multiply(A=@n64, B=@n66), B=AppendVector(A=@n67.Random_offset, B=@n68)), B=AppendVector(A=Constant() {r=0.5}, B=Constant()))
  n70 = TextureSampleParameter2D(UVs=Panner(Coordinate=Add(A=@n69, B=@n33)) {speedX=-0.25}) {parameterName=Texture_C, texture=&T_TechLines}
  n71 = Multiply(A=Power(Base=Multiply(A=@n70.R) {constB=5}, Exp=ScalarParameter() {defaultValue=1, parameterName=Texture_C_power}), B=@n36)
  n72 = LinearInterpolate(A=Multiply(A=Add(A=@n63, B=@n71), B=VertexColor().G), B=Constant(), Alpha=ScalarParameter() {parameterName=Only_Use_one_wing})
  n73 = TextureCoordinate()
  n74 = Add(A=ComponentMask(None=@n73) {r=true}, B=ScalarParameter() {defaultValue=-0.06, parameterName=Bottom_mask_offset}) {constB=-0.06}
  n75 = OneMinus(None=ComponentMask(None=@n73) {r=true})
  n76 = ComponentMask(None=TextureCoordinate()) {g=true}
  n77 = Multiply(A=@n76, B=OneMinus(None=@n76))
  n78 = Multiply(A=Multiply(A=Add(A=@n40, B=@n72) {constB=0}, B=Multiply(A=@n74, B=@n75)), B=Power(Base=Multiply(A=@n77) {constB=3}))
  n79 = ComponentMask(None=TextureCoordinate()) {r=true}
  n80 = Power(Base=Multiply(A=OneMinus(None=@n79)) {constB=3}) {constExponent=1.5}
  n81 = TextureCoordinate()
  n82 = VectorParameter() {group=Trailing Edge Sphere Mask, parameterName=Edge_mask_Sphereical_UV_Scale, sortPriority=4, defaultValue=<rgba 1 0.8 0 0>}
  n83 = ComponentMask(None=@n82.RGB) {g=true, r=true}
  n84 = VectorParameter() {group=Trailing Edge Sphere Mask, parameterName="Edge_mask_Sphere_Mask_Coord (B)", sortPriority=3, defaultValue=<rgba 1.8 0.7 0 0>}
  n85 = OneMinus(None=SphereMask(
    A=Multiply(A=@n81, B=@n83)
    B=ComponentMask(None=@n84.RGB) {g=true, r=true}
    Radius=ScalarParameter() {defaultValue=1.1, group=Trailing Edge Sphere Mask, parameterName=Edge_mask_Radius, sortPriority=1}
    Hardness=ScalarParameter() {defaultValue=0.8, group=Trailing Edge Sphere Mask, parameterName=Edge_mask_Hardness, sortPriority=2}
   ))
  n86 = ScalarParameter() {defaultValue=10, group=Trailing Edge Sphere Mask, parameterName=Edge_mask_Contrast}
  n87 = Multiply(
   A=Multiply(A=Multiply(A=@n8.A, B=@n78), B=Saturate(None=@n80))
   B=LinearInterpolate(
    B=Power(Base=@n85, Exp=@n86) {constExponent=5}
    Alpha=ScalarParameter() {defaultValue=1, group=Alpha, parameterName=Edge_mask_Sphere_Mask, sortPriority=6}
   ) {constA=1}
  )
  n88 = DepthFade(Opacity=@n87, FadeDistance=DynamicParameter() {defaultValue=<rgba 0 1 1 10>, paramNames [Random_offset, DepthFade, Param3, WPO]}.DepthFade)
  n89 = Multiply(A=@n87) {constB=0.4}
  ShadingPathSwitch(Default=@n88, Deferred=@n88, Forward=@n89, Mobile=@n89)
 ]
}
